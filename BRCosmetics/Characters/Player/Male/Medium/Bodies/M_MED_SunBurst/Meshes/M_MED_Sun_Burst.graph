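# Physics asset for the Sun Burst medium male player body, covering the necklace bones and the upper spine.
# A small sphere on the necklace master bone is constrained to a large spine_05 capsule with free swing and twist.
# Two capsules on the dynamic necklace bones chain from the master bone with tight angular limits, allowing slight secondary motion.
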PhysicsAsset /BRCosmetics/Characters/Player/Male/Medium/Bodies/M_MED_SunBurst/Meshes/M_MED_Sun_Burst
bodies {
 neclace_01_dyn [{shapeType=Capsule, length=5, radius=3, shapeName=neclace_01_dyn_capsule, center=<2 0 2.25>, rotation=<rot 0 0 90>}]
 neclace_02_dyn [{shapeType=Capsule, length=3, radius=3, shapeName=neclace_02_dyn_capsule, center=<4 0 1.8>, rotation=<rot 90 0 0>}]
 neclace_mstr [{shapeType=Sphere, radius=1, shapeName=neclace_mstr_sphere}]
 spine_05 [{shapeType=Capsule, length=16, radius=9.3, shapeName=spine_05_capsule, center=<0 0.75 0>}]
}
constraints [
 {
  bone1Name=neclace_mstr
  bone2Name=spine_05
  swing1LimitDegrees=45
  swing1Motion=Free
  swing2LimitDegrees=45
  swing2Motion=Free
  twistLimitDegrees=45
  twistMotion=Free
 }
 {bone1Name=neclace_01_dyn, bone2Name=neclace_mstr, swing1LimitDegrees=15, swing2LimitDegrees=10, twistLimitDegrees=5}
 {bone1Name=neclace_02_dyn, bone2Name=neclace_01_dyn, swing1LimitDegrees=15, swing2LimitDegrees=5, twistLimitDegrees=5}
]
