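# Animation Blueprint for the male medium SunBurst cosmetic body, layering secondary motion onto the pose copied from its attached parent mesh.
# A deform Control Rig feeds a RigidBody simulation, then chained spring-based AnimDynamics nodes for shirt sleeves and shoelaces, with spherical limits on laces.
# A second Control Rig finishes the pose; the EventGraph has no logic.
# One right-foot node binds foot_l_shoelace_l_02_dyn again, leaving foot_r_shoelace_r_02_dyn without a node.
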AnimBlueprint /BRCosmetics/Characters/Player/Male/Medium/Bodies/M_MED_SunBurst/Meshes/M_MED_SunBurst_AnimBP
target_skeleton=&/M_MED_SunBurst_Skeleton
graphs [
 AnimGraph {nodes [
   n1 = Control Rig(Source=Copy Pose From Mesh() {bUseAttachedParent=true}) {
    alphaBoolBlend {blendOption=Linear}
    controlRigAssetReference {blueprintRigClass=&/M_MED_Sun_Burst_DeformRig_Master_ControlRig.M_MED_Sun_Burst_DeformRig_Master_ControlRig_C}
   }
   n2 = RigidBody(ComponentPose=Local To Component(LocalPose=@n1)) {
    bClampLinearTranslationLimitToRefPose=true
    bOverrideWorldGravity=true
    overridePhysicsAsset=&/M_MED_Sun_Burst
    simulationSpace=BaseBoneSpace
    alphaBoolBlend {blendOption=Linear}
    baseBoneRef {boneName=root}
    componentAppliedLinearAccClamp=<3000 3000 3000>
    componentLinearAccScale=<0.3 0.3 0.45>
    componentLinearVelScale=<0.15 0.15 0.225>
    overrideWorldGravity {z=-980}
   }
   n3 = Anim Dynamics - Bone: shirt_sleeve_front_l(ComponentPose=@n2, Alpha=1, ExternalForce="0,0,0") {
    angularDampingOverride=0.9999
    angularSpringConstant=75
    bAngularSpring=true
    bGravityOverrideInSimSpace=true
    bOverrideAngularDamping=true
    simulationSpace=BoneRelative
    alphaBoolBlend {blendOption=Linear}
    boundBone {boneName=shirt_sleeve_front_l}
    componentLinearAccScale=<1 1 1>
    componentLinearVelScale=<1 1 1>
    physicsBodyDefinitions [{
      sphereCollisionRadius=0
      boundBone {boneName=shirt_sleeve_front_l}
      boxExtents=<10 20 20>
      constraintSetup {angularLimitsMax=<15 15 15>, angularLimitsMin=<-15 -15 -15>, angularTarget=<0 0 0>}
      localJointOffset=<1 0 -0.2>
     }]
    relativeSpaceBone {boneName=upperarm_l}
    sphericalLimits [{limitRadius=4, drivingBone {boneName=shoe_knot_l_leg_mstr}, sphereLocalOffset=<-0.5 4 0>}]
   }
   n4 = Anim Dynamics - Bone: shirt_sleeve_back_l(ComponentPose=@n3, Alpha=1, ExternalForce="0,0,0") {
    angularDampingOverride=0.9999
    angularSpringConstant=75
    bAngularSpring=true
    bGravityOverrideInSimSpace=true
    bOverrideAngularDamping=true
    simulationSpace=BoneRelative
    alphaBoolBlend {blendOption=Linear}
    boundBone {boneName=shirt_sleeve_back_l}
    componentLinearAccScale=<1 1 1>
    componentLinearVelScale=<1 1 1>
    physicsBodyDefinitions [{
      sphereCollisionRadius=0
      boundBone {boneName=shirt_sleeve_back_l}
      boxExtents=<10 20 20>
      constraintSetup {angularLimitsMax=<15 15 15>, angularLimitsMin=<-15 -15 -15>, angularTarget=<0 0 0>}
      localJointOffset=<1 0 -0.2>
     }]
    relativeSpaceBone {boneName=upperarm_l}
    sphericalLimits [{limitRadius=4, drivingBone {boneName=shoe_knot_l_leg_mstr}, sphereLocalOffset=<-0.5 4 0>}]
   }
   n5 = Anim Dynamics - Bone: shirt_sleeve_front_r(ComponentPose=@n4, Alpha=1, ExternalForce="0,0,0") {
    angularDampingOverride=0.9999
    angularSpringConstant=75
    bAngularSpring=true
    bGravityOverrideInSimSpace=true
    bOverrideAngularDamping=true
    simulationSpace=BoneRelative
    alphaBoolBlend {blendOption=Linear}
    boundBone {boneName=shirt_sleeve_front_r}
    componentLinearAccScale=<1 1 1>
    componentLinearVelScale=<1 1 1>
    physicsBodyDefinitions [{
      sphereCollisionRadius=0
      boundBone {boneName=shirt_sleeve_front_r}
      boxExtents=<10 20 20>
      constraintSetup {angularLimitsMax=<15 15 15>, angularLimitsMin=<-15 -15 -15>, angularTarget=<0 0 0>}
      localJointOffset=<1 0 -0.2>
     }]
    relativeSpaceBone {boneName=upperarm_r}
    sphericalLimits [{limitRadius=4, drivingBone {boneName=shoe_knot_l_leg_mstr}, sphereLocalOffset=<-0.5 4 0>}]
   }
   n6 = Anim Dynamics - Bone: shirt_sleeve_back_r(ComponentPose=@n5, Alpha=1, ExternalForce="0,0,0") {
    angularDampingOverride=0.9999
    angularSpringConstant=75
    bAngularSpring=true
    bGravityOverrideInSimSpace=true
    bOverrideAngularDamping=true
    simulationSpace=BoneRelative
    alphaBoolBlend {blendOption=Linear}
    boundBone {boneName=shirt_sleeve_back_r}
    componentLinearAccScale=<1 1 1>
    componentLinearVelScale=<1 1 1>
    physicsBodyDefinitions [{
      sphereCollisionRadius=0
      boundBone {boneName=shirt_sleeve_back_r}
      boxExtents=<10 20 20>
      constraintSetup {angularLimitsMax=<15 15 15>, angularLimitsMin=<-15 -15 -15>, angularTarget=<0 0 0>}
      localJointOffset=<1 0 -0.2>
     }]
    relativeSpaceBone {boneName=upperarm_r}
    sphericalLimits [{limitRadius=4, drivingBone {boneName=shoe_knot_l_leg_mstr}, sphereLocalOffset=<-0.5 4 0>}]
   }
   n7 = Anim Dynamics - Bone: foot_l_shoelace_l_01_dyn(ComponentPose=@n6, Alpha=1, ExternalForce="0,0,0") {
    angularDampingOverride=0.999
    angularSpringConstant=75
    bAngularSpring=true
    bGravityOverrideInSimSpace=true
    bOverrideAngularDamping=true
    bUseSphericalLimits=true
    simulationSpace=BoneRelative
    alphaBoolBlend {blendOption=Linear}
    boundBone {boneName=foot_l_shoelace_l_01_dyn}
    componentLinearAccScale=<1 1 1>
    componentLinearVelScale=<1 1 1>
    physicsBodyDefinitions [{
      sphereCollisionRadius=0
      boundBone {boneName=foot_l_shoelace_l_01_dyn}
      boxExtents=<10 20 20>
      constraintSetup {angularLimitsMax=<45 45 45>, angularLimitsMin=<-45 -45 -45>, angularTarget=<0 0 0>}
      localJointOffset=<1 0 -0.2>
     }]
    relativeSpaceBone {boneName=foot_l}
    sphericalLimits [{limitRadius=4.5, drivingBone {boneName=shoe_knot_l_leg_mstr}, sphereLocalOffset=<-0.5 4 0>}]
   }
   n8 = Anim Dynamics - Bone: foot_l_shoelace_l_02_dyn(ComponentPose=@n7, Alpha=1, ExternalForce="0,0,0") {
    angularDampingOverride=0.999
    angularSpringConstant=75
    bAngularSpring=true
    bGravityOverrideInSimSpace=true
    bOverrideAngularDamping=true
    bUseSphericalLimits=true
    simulationSpace=BoneRelative
    alphaBoolBlend {blendOption=Linear}
    boundBone {boneName=foot_l_shoelace_l_02_dyn}
    componentLinearAccScale=<1 1 1>
    componentLinearVelScale=<1 1 1>
    physicsBodyDefinitions [{
      sphereCollisionRadius=0
      boundBone {boneName=foot_l_shoelace_l_02_dyn}
      boxExtents=<10 20 20>
      constraintSetup {angularLimitsMax=<45 45 45>, angularLimitsMin=<-45 -45 -45>, angularTarget=<0 0 0>}
      localJointOffset=<1 0 -0.2>
     }]
    relativeSpaceBone {boneName=foot_l}
    sphericalLimits [{limitRadius=4.5, drivingBone {boneName=shoe_knot_l_leg_mstr}, sphereLocalOffset=<-0.5 4 0>}]
   }
   n9 = Anim Dynamics - Bone: foot_l_shoelace_r_01_dyn(ComponentPose=@n8, Alpha=1, ExternalForce="0,0,0") {
    angularDampingOverride=0.999
    angularSpringConstant=75
    bAngularSpring=true
    bGravityOverrideInSimSpace=true
    bOverrideAngularDamping=true
    bUseSphericalLimits=true
    simulationSpace=BoneRelative
    alphaBoolBlend {blendOption=Linear}
    boundBone {boneName=foot_l_shoelace_r_01_dyn}
    componentLinearAccScale=<1 1 1>
    componentLinearVelScale=<1 1 1>
    physicsBodyDefinitions [{
      sphereCollisionRadius=0
      boundBone {boneName=foot_l_shoelace_r_01_dyn}
      boxExtents=<10 20 20>
      constraintSetup {angularLimitsMax=<45 45 45>, angularLimitsMin=<-45 -45 -45>, angularTarget=<0 0 0>}
      localJointOffset=<1 0 -0.2>
     }]
    relativeSpaceBone {boneName=foot_l}
    sphericalLimits [{limitRadius=4.5, drivingBone {boneName=shoe_knot_l_leg_mstr}, sphereLocalOffset=<-0.5 4 0>}]
   }
   n10 = Anim Dynamics - Bone: foot_l_shoelace_r_02_dyn(ComponentPose=@n9, Alpha=1, ExternalForce="0,0,0") {
    angularDampingOverride=0.999
    angularSpringConstant=75
    bAngularSpring=true
    bGravityOverrideInSimSpace=true
    bOverrideAngularDamping=true
    bUseSphericalLimits=true
    simulationSpace=BoneRelative
    alphaBoolBlend {blendOption=Linear}
    boundBone {boneName=foot_l_shoelace_r_02_dyn}
    componentLinearAccScale=<1 1 1>
    componentLinearVelScale=<1 1 1>
    physicsBodyDefinitions [{
      sphereCollisionRadius=0
      boundBone {boneName=foot_l_shoelace_r_02_dyn}
      boxExtents=<10 20 20>
      constraintSetup {angularLimitsMax=<45 45 45>, angularLimitsMin=<-45 -45 -45>, angularTarget=<0 0 0>}
      localJointOffset=<1 0 -0.2>
     }]
    relativeSpaceBone {boneName=foot_l}
    sphericalLimits [{limitRadius=4.5, drivingBone {boneName=shoe_knot_l_leg_mstr}, sphereLocalOffset=<-0.5 4 0>}]
   }
   n11 = Anim Dynamics - Bone: foot_r_shoelace_r_01_dyn(ComponentPose=@n10, Alpha=1, ExternalForce="0,0,0") {
    angularDampingOverride=0.999
    angularSpringConstant=75
    bAngularSpring=true
    bGravityOverrideInSimSpace=true
    bOverrideAngularDamping=true
    bUseSphericalLimits=true
    simulationSpace=BoneRelative
    alphaBoolBlend {blendOption=Linear}
    boundBone {boneName=foot_r_shoelace_r_01_dyn}
    componentLinearAccScale=<1 1 1>
    componentLinearVelScale=<1 1 1>
    physicsBodyDefinitions [{
      sphereCollisionRadius=0
      boundBone {boneName=foot_r_shoelace_r_01_dyn}
      boxExtents=<10 20 20>
      constraintSetup {angularLimitsMax=<45 45 45>, angularLimitsMin=<-45 -45 -45>, angularTarget=<0 0 0>}
      localJointOffset=<1 0 -0.2>
     }]
    relativeSpaceBone {boneName=foot_r}
    sphericalLimits [{limitRadius=4.5, drivingBone {boneName=shoe_knot_r_leg_mstr}, sphereLocalOffset=<-0.5 -4 0>}]
   }
   n12 = Anim Dynamics - Bone: foot_l_shoelace_l_02_dyn(ComponentPose=@n11, Alpha=1, ExternalForce="0,0,0") {
    angularDampingOverride=0.999
    angularSpringConstant=75
    bAngularSpring=true
    bGravityOverrideInSimSpace=true
    bOverrideAngularDamping=true
    bUseSphericalLimits=true
    simulationSpace=BoneRelative
    alphaBoolBlend {blendOption=Linear}
    boundBone {boneName=foot_l_shoelace_l_02_dyn}
    componentLinearAccScale=<1 1 1>
    componentLinearVelScale=<1 1 1>
    physicsBodyDefinitions [{
      sphereCollisionRadius=0
      boundBone {boneName=foot_l_shoelace_l_02_dyn}
      boxExtents=<10 20 20>
      constraintSetup {angularLimitsMax=<45 45 45>, angularLimitsMin=<-45 -45 -45>, angularTarget=<0 0 0>}
      localJointOffset=<1 0 -0.2>
     }]
    relativeSpaceBone {boneName=foot_r}
    sphericalLimits [{limitRadius=4.5, drivingBone {boneName=shoe_knot_r_leg_mstr}, sphereLocalOffset=<-0.5 -4 0>}]
   }
   n13 = Anim Dynamics - Bone: foot_r_shoelace_l_01_dyn(ComponentPose=@n12, Alpha=1, ExternalForce="0,0,0") {
    angularDampingOverride=0.999
    angularSpringConstant=75
    bAngularSpring=true
    bGravityOverrideInSimSpace=true
    bOverrideAngularDamping=true
    bUseSphericalLimits=true
    simulationSpace=BoneRelative
    alphaBoolBlend {blendOption=Linear}
    boundBone {boneName=foot_r_shoelace_l_01_dyn}
    componentLinearAccScale=<1 1 1>
    componentLinearVelScale=<1 1 1>
    physicsBodyDefinitions [{
      sphereCollisionRadius=0
      boundBone {boneName=foot_r_shoelace_l_01_dyn}
      boxExtents=<10 20 20>
      constraintSetup {angularLimitsMax=<45 45 45>, angularLimitsMin=<-45 -45 -45>, angularTarget=<0 0 0>}
      localJointOffset=<1 0 -0.2>
     }]
    relativeSpaceBone {boneName=foot_r}
    sphericalLimits [{limitRadius=4.5, drivingBone {boneName=shoe_knot_r_leg_mstr}, sphereLocalOffset=<-0.5 -4 0>}]
   }
   n14 = Anim Dynamics - Bone: foot_r_shoelace_l_02_dyn(ComponentPose=@n13, Alpha=1, ExternalForce="0,0,0") {
    angularDampingOverride=0.999
    angularSpringConstant=75
    bAngularSpring=true
    bGravityOverrideInSimSpace=true
    bOverrideAngularDamping=true
    bUseSphericalLimits=true
    simulationSpace=BoneRelative
    alphaBoolBlend {blendOption=Linear}
    boundBone {boneName=foot_r_shoelace_l_02_dyn}
    componentLinearAccScale=<1 1 1>
    componentLinearVelScale=<1 1 1>
    physicsBodyDefinitions [{
      sphereCollisionRadius=0
      boundBone {boneName=foot_r_shoelace_l_02_dyn}
      boxExtents=<10 20 20>
      constraintSetup {angularLimitsMax=<45 45 45>, angularLimitsMin=<-45 -45 -45>, angularTarget=<0 0 0>}
      localJointOffset=<1 0 -0.2>
     }]
    relativeSpaceBone {boneName=foot_r}
    sphericalLimits [{limitRadius=4.5, drivingBone {boneName=shoe_knot_r_leg_mstr}, sphereLocalOffset=<-0.5 -4 0>}]
   }
   n15 = Control Rig(Source=Component To Local(ComponentPose=@n14)) {
    alphaBoolBlend {blendOption=Linear}
    controlRigAssetReference {blueprintRigClass=&/M_MED_SunBurst_Control_Rig.M_MED_SunBurst_Control_Rig_C}
   }
   Output Pose(Result=@n15) {layerGroup=None}
  ]}
 EventGraph {nodes [Event BlueprintUpdateAnimation()]}
]
variables []
