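# Physics asset for the SunBurst medium male hat face accessory, defining bodies for a dangling rope and the upper body around it.
# A large head sphere and upper-spine capsule provide body volume, joined through two tiny neck spheres.
# The rope chain runs from a master sphere on the head through three small capsules.
# Constraints rotate freely down to the first rope segment, while the two lower rope joints have tighter twist than swing limits.
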PhysicsAsset /BRCosmetics/Characters/Player/Male/Medium/Bodies/M_MED_SunBurst/Meshes/Parts/M_Med_SunBurst_Hat_FaceAcc_PA
bodies {
 head [{shapeType=Sphere, radius=17.2027, shapeName=head_sphere, center=<7.1489 0.8106 -0.2638>}]
 neck_01 [{shapeType=Sphere, radius=0.505, shapeName=neck_01_sphere}]
 neck_02 [{shapeType=Sphere, radius=0.505, shapeName=neck_02_sphere}]
 rope_dyn_01 [{shapeType=Capsule, length=2, radius=1, shapeName=rope_dyn_01_capsule, center=<2 0 0>, rotation=<rot 90 0 0>}]
 rope_dyn_02 [{shapeType=Capsule, length=2, radius=1, shapeName=rope_dyn_02_capsule, center=<2 0 0>, rotation=<rot 90 0 0>}]
 rope_dyn_03 [{shapeType=Capsule, length=6, radius=1, shapeName=rope_dyn_03_capsule, center=<3.5 0 0>, rotation=<rot 90 0 0>}]
 rope_dyn_mstr [{shapeType=Sphere, radius=1, shapeName=rope_dyn_mstr_sphere}]
 spine_05 [{shapeType=Capsule, length=15, radius=9.5, shapeName=spine_05_capsule, center=<0 2.5 0>}]
}
constraints [
 {
  bone1Name=neck_01
  bone2Name=spine_05
  swing1LimitDegrees=45
  swing1Motion=Free
  swing2LimitDegrees=45
  swing2Motion=Free
  twistLimitDegrees=45
  twistMotion=Free
 }
 {
  bone1Name=neck_02
  bone2Name=neck_01
  swing1LimitDegrees=45
  swing1Motion=Free
  swing2LimitDegrees=45
  swing2Motion=Free
  twistLimitDegrees=45
  twistMotion=Free
 }
 {
  bone1Name=head
  bone2Name=neck_02
  swing1LimitDegrees=45
  swing1Motion=Free
  swing2LimitDegrees=45
  swing2Motion=Free
  twistLimitDegrees=45
  twistMotion=Free
 }
 {bone1Name=rope_dyn_02, bone2Name=rope_dyn_01, swing1LimitDegrees=45, swing2LimitDegrees=45, twistLimitDegrees=5}
 {bone1Name=rope_dyn_03, bone2Name=rope_dyn_02, swing1LimitDegrees=45, swing2LimitDegrees=45, twistLimitDegrees=5}
 {
  bone1Name=rope_dyn_01
  bone2Name=rope_dyn_mstr
  swing1LimitDegrees=45
  swing1Motion=Free
  swing2LimitDegrees=45
  swing2Motion=Free
  twistLimitDegrees=45
  twistMotion=Free
 }
 {
  bone1Name=rope_dyn_mstr
  bone2Name=head
  swing1LimitDegrees=45
  swing1Motion=Free
  swing2LimitDegrees=45
  swing2Motion=Free
  twistLimitDegrees=45
  twistMotion=Free
 }
]
