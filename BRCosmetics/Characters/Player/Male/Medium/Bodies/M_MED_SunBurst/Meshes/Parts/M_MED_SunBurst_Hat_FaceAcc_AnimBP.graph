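# Post-processes the pose of the male medium SunBurst hat/face-accessory cosmetic mesh, adding secondary motion to its hair.
# Copies the attached parent's pose, then runs a Control Rig and a RigidBody simulation using an override physics asset in root-bone space.
# Per-bone AnimDynamics nodes then run in series over eight three-bone bang and hair strands, using damped angular springs, angular limits and a head-driven spherical limit.
# The event graph has no logic.
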AnimBlueprint /BRCosmetics/Characters/Player/Male/Medium/Bodies/M_MED_SunBurst/Meshes/Parts/M_MED_SunBurst_Hat_FaceAcc_AnimBP
target_skeleton=&/M_MED_SunBurst_Hat_FaceAcc_Skeleton
graphs [
 AnimGraph {nodes [
   n1 = Control Rig(Source=Copy Pose From Mesh() {bUseAttachedParent=true}) {
    alphaBoolBlend {blendOption=Linear}
    controlRigAssetReference {blueprintRigClass=&/M_MED_SunBurst_Hat_FaceAcc_CtrlRig.M_MED_SunBurst_Hat_FaceAcc_CtrlRig_C}
   }
   n2 = RigidBody(ComponentPose=Local To Component(LocalPose=@n1)) {
    bClampLinearTranslationLimitToRefPose=true
    bOverrideWorldGravity=true
    overridePhysicsAsset=&/M_Med_SunBurst_Hat_FaceAcc_PA
    simulationSpace=BaseBoneSpace
    alphaBoolBlend {blendOption=Linear}
    baseBoneRef {boneName=root}
    componentAppliedLinearAccClamp=<3000 3000 3000>
    componentLinearAccScale=<0.3 0.3 0.45>
    componentLinearVelScale=<0.15 0.15 0.225>
    overrideWorldGravity {z=-980}
   }
   n3 = Anim Dynamics - Bone: bang_r_01_dyn(ComponentPose=@n2, Alpha=1, ExternalForce="0,0,0") {
    angularDampingOverride=0.99
    angularSpringConstant=75
    bAngularSpring=true
    bGravityOverrideInSimSpace=true
    bOverrideAngularDamping=true
    bUseSphericalLimits=true
    simulationSpace=RootRelative
    alphaBoolBlend {blendOption=Linear}
    boundBone {boneName=bang_r_01_dyn}
    componentLinearAccScale=<1 2 1>
    componentLinearVelScale=<1 2 1>
    physicsBodyDefinitions [{
      sphereCollisionRadius=0
      boundBone {boneName=bang_r_01_dyn}
      boxExtents=<10 20 20>
      constraintSetup {angularLimitsMax=<45 5 30>, angularLimitsMin=<-45 -20 0>, angularTarget=<0 0 0>}
      localJointOffset=<1 0 -0.2>
     }]
    sphericalLimits [{limitRadius=6.5, drivingBone {boneName=head}, sphereLocalOffset=<4 0 0>}]
   }
   n4 = Anim Dynamics - Bone: bang_r_02_dyn(ComponentPose=@n3, Alpha=1, ExternalForce="0,0,0") {
    angularDampingOverride=0.99
    angularSpringConstant=75
    bAngularSpring=true
    bGravityOverrideInSimSpace=true
    bOverrideAngularDamping=true
    bUseSphericalLimits=true
    simulationSpace=RootRelative
    alphaBoolBlend {blendOption=Linear}
    boundBone {boneName=bang_r_02_dyn}
    componentLinearAccScale=<1 2 1>
    componentLinearVelScale=<1 2 1>
    physicsBodyDefinitions [{
      sphereCollisionRadius=0
      boundBone {boneName=bang_r_02_dyn}
      boxExtents=<10 20 20>
      constraintSetup {angularLimitsMax=<45 5 30>, angularLimitsMin=<-45 -20 0>, angularTarget=<0 0 0>}
      localJointOffset=<1 0 -0.2>
     }]
    sphericalLimits [{limitRadius=6.5, drivingBone {boneName=head}, sphereLocalOffset=<4 0 0>}]
   }
   n5 = Anim Dynamics - Bone: bang_r_03_dyn(ComponentPose=@n4, Alpha=1, ExternalForce="0,0,0") {
    angularDampingOverride=0.99
    angularSpringConstant=75
    bAngularSpring=true
    bGravityOverrideInSimSpace=true
    bOverrideAngularDamping=true
    bUseSphericalLimits=true
    simulationSpace=RootRelative
    alphaBoolBlend {blendOption=Linear}
    boundBone {boneName=bang_r_03_dyn}
    componentLinearAccScale=<1 2 1>
    componentLinearVelScale=<1 2 1>
    physicsBodyDefinitions [{
      sphereCollisionRadius=0
      boundBone {boneName=bang_r_03_dyn}
      boxExtents=<10 20 20>
      constraintSetup {angularLimitsMax=<45 5 30>, angularLimitsMin=<-45 -20 0>, angularTarget=<0 0 0>}
      localJointOffset=<1 0 -0.2>
     }]
    sphericalLimits [{limitRadius=6.5, drivingBone {boneName=head}, sphereLocalOffset=<4 0 0>}]
   }
   n6 = Anim Dynamics - Bone: hair_mid_fr_01_dyn(ComponentPose=@n5, Alpha=1, ExternalForce="0,0,0") {
    angularDampingOverride=0.99
    angularSpringConstant=75
    bAngularSpring=true
    bGravityOverrideInSimSpace=true
    bOverrideAngularDamping=true
    bUseSphericalLimits=true
    simulationSpace=RootRelative
    alphaBoolBlend {blendOption=Linear}
    boundBone {boneName=hair_mid_fr_01_dyn}
    componentLinearAccScale=<1 2 1>
    componentLinearVelScale=<1 2 1>
    physicsBodyDefinitions [{
      sphereCollisionRadius=0
      boundBone {boneName=hair_mid_fr_01_dyn}
      boxExtents=<10 20 20>
      constraintSetup {angularLimitsMax=<45 45 15>, angularLimitsMin=<-45 -45 -45>, angularTarget=<0 0 0>}
      localJointOffset=<1 0 -0.2>
     }]
    sphericalLimits [{limitRadius=6.5, drivingBone {boneName=head}, sphereLocalOffset=<4 0 0>}]
   }
   n7 = Anim Dynamics - Bone: hair_mid_fr_02_dyn(ComponentPose=@n6, Alpha=1, ExternalForce="0,0,0") {
    angularDampingOverride=0.99
    angularSpringConstant=75
    bAngularSpring=true
    bGravityOverrideInSimSpace=true
    bOverrideAngularDamping=true
    bUseSphericalLimits=true
    simulationSpace=RootRelative
    alphaBoolBlend {blendOption=Linear}
    boundBone {boneName=hair_mid_fr_02_dyn}
    componentLinearAccScale=<1 2 1>
    componentLinearVelScale=<1 2 1>
    physicsBodyDefinitions [{
      sphereCollisionRadius=0
      boundBone {boneName=hair_mid_fr_02_dyn}
      boxExtents=<10 20 20>
      constraintSetup {angularLimitsMax=<45 45 15>, angularLimitsMin=<-45 -45 -45>, angularTarget=<0 0 0>}
      localJointOffset=<1 0 -0.2>
     }]
    sphericalLimits [{limitRadius=6.5, drivingBone {boneName=head}, sphereLocalOffset=<4 0 0>}]
   }
   n8 = Anim Dynamics - Bone: hair_mid_fr_03_dyn(ComponentPose=@n7, Alpha=1, ExternalForce="0,0,0") {
    angularDampingOverride=0.99
    angularSpringConstant=75
    bAngularSpring=true
    bGravityOverrideInSimSpace=true
    bOverrideAngularDamping=true
    bUseSphericalLimits=true
    simulationSpace=RootRelative
    alphaBoolBlend {blendOption=Linear}
    boundBone {boneName=hair_mid_fr_03_dyn}
    componentLinearAccScale=<1 2 1>
    componentLinearVelScale=<1 2 1>
    physicsBodyDefinitions [{
      sphereCollisionRadius=0
      boundBone {boneName=hair_mid_fr_03_dyn}
      boxExtents=<10 20 20>
      constraintSetup {angularLimitsMax=<45 45 15>, angularLimitsMin=<-45 -45 -45>, angularTarget=<0 0 0>}
      localJointOffset=<1 0 -0.2>
     }]
    sphericalLimits [{limitRadius=6.5, drivingBone {boneName=head}, sphereLocalOffset=<4 0 0>}]
   }
   n9 = Anim Dynamics - Bone: hair_mid_br_01_dyn(ComponentPose=@n8, Alpha=1, ExternalForce="0,0,0") {
    angularDampingOverride=0.99
    angularSpringConstant=75
    bAngularSpring=true
    bGravityOverrideInSimSpace=true
    bOverrideAngularDamping=true
    bUseSphericalLimits=true
    simulationSpace=RootRelative
    alphaBoolBlend {blendOption=Linear}
    boundBone {boneName=hair_mid_br_01_dyn}
    componentLinearAccScale=<1 1.5 1>
    componentLinearVelScale=<1 1.5 1>
    physicsBodyDefinitions [{
      sphereCollisionRadius=0
      boundBone {boneName=hair_mid_br_01_dyn}
      boxExtents=<10 20 20>
      constraintSetup {angularLimitsMax=<45 45 15>, angularLimitsMin=<-45 -45 -45>, angularTarget=<0 0 0>}
      localJointOffset=<1 0 -0.2>
     }]
    sphericalLimits [{limitRadius=6.5, drivingBone {boneName=head}, sphereLocalOffset=<4 0 0>}]
   }
   n10 = Anim Dynamics - Bone: hair_mid_br_02_dyn(ComponentPose=@n9, Alpha=1, ExternalForce="0,0,0") {
    angularDampingOverride=0.99
    angularSpringConstant=75
    bAngularSpring=true
    bGravityOverrideInSimSpace=true
    bOverrideAngularDamping=true
    bUseSphericalLimits=true
    simulationSpace=RootRelative
    alphaBoolBlend {blendOption=Linear}
    boundBone {boneName=hair_mid_br_02_dyn}
    componentLinearAccScale=<1 1.5 1>
    componentLinearVelScale=<1 1.5 1>
    physicsBodyDefinitions [{
      sphereCollisionRadius=0
      boundBone {boneName=hair_mid_br_02_dyn}
      boxExtents=<10 20 20>
      constraintSetup {angularLimitsMax=<45 45 15>, angularLimitsMin=<-45 -45 -45>, angularTarget=<0 0 0>}
      localJointOffset=<1 0 -0.2>
     }]
    sphericalLimits [{limitRadius=6.5, drivingBone {boneName=head}, sphereLocalOffset=<4 0 0>}]
   }
   n11 = Anim Dynamics - Bone: hair_mid_br_03_dyn(ComponentPose=@n10, Alpha=1, ExternalForce="0,0,0") {
    angularDampingOverride=0.99
    angularSpringConstant=75
    bAngularSpring=true
    bGravityOverrideInSimSpace=true
    bOverrideAngularDamping=true
    bUseSphericalLimits=true
    simulationSpace=RootRelative
    alphaBoolBlend {blendOption=Linear}
    boundBone {boneName=hair_mid_br_03_dyn}
    componentLinearAccScale=<1 1.5 1>
    componentLinearVelScale=<1 1.5 1>
    physicsBodyDefinitions [{
      sphereCollisionRadius=0
      boundBone {boneName=hair_mid_br_03_dyn}
      boxExtents=<10 20 20>
      constraintSetup {angularLimitsMax=<45 45 15>, angularLimitsMin=<-45 -45 -45>, angularTarget=<0 0 0>}
      localJointOffset=<1 0 -0.2>
     }]
    sphericalLimits [{limitRadius=6.5, drivingBone {boneName=head}, sphereLocalOffset=<4 0 0>}]
   }
   n12 = Anim Dynamics - Bone: hair_br_01_dyn(ComponentPose=@n11, Alpha=1, ExternalForce="0,0,0") {
    angularDampingOverride=0.99
    angularSpringConstant=75
    bAngularSpring=true
    bGravityOverrideInSimSpace=true
    bOverrideAngularDamping=true
    bUseSphericalLimits=true
    simulationSpace=RootRelative
    alphaBoolBlend {blendOption=Linear}
    boundBone {boneName=hair_br_01_dyn}
    componentLinearAccScale=<1 1 1>
    componentLinearVelScale=<1 1 1>
    physicsBodyDefinitions [{
      sphereCollisionRadius=0
      boundBone {boneName=hair_br_01_dyn}
      boxExtents=<10 20 20>
      constraintSetup {angularLimitsMax=<45 15 45>, angularLimitsMin=<-45 -15 -45>, angularTarget=<0 0 0>}
      localJointOffset=<1 0 -0.2>
     }]
    sphericalLimits [{limitRadius=6.5, drivingBone {boneName=head}, sphereLocalOffset=<4 0 0>}]
   }
   n13 = Anim Dynamics - Bone: hair_br_02_dyn(ComponentPose=@n12, Alpha=1, ExternalForce="0,0,0") {
    angularDampingOverride=0.99
    angularSpringConstant=75
    bAngularSpring=true
    bGravityOverrideInSimSpace=true
    bOverrideAngularDamping=true
    bUseSphericalLimits=true
    simulationSpace=RootRelative
    alphaBoolBlend {blendOption=Linear}
    boundBone {boneName=hair_br_02_dyn}
    componentLinearAccScale=<1 1 1>
    componentLinearVelScale=<1 1 1>
    physicsBodyDefinitions [{
      sphereCollisionRadius=0
      boundBone {boneName=hair_br_02_dyn}
      boxExtents=<10 20 20>
      constraintSetup {angularLimitsMax=<45 15 45>, angularLimitsMin=<-45 -15 -45>, angularTarget=<0 0 0>}
      localJointOffset=<1 0 -0.2>
     }]
    sphericalLimits [{limitRadius=6.5, drivingBone {boneName=head}, sphereLocalOffset=<4 0 0>}]
   }
   n14 = Anim Dynamics - Bone: hair_br_03_dyn(ComponentPose=@n13, Alpha=1, ExternalForce="0,0,0") {
    angularDampingOverride=0.99
    angularSpringConstant=75
    bAngularSpring=true
    bGravityOverrideInSimSpace=true
    bOverrideAngularDamping=true
    bUseSphericalLimits=true
    simulationSpace=RootRelative
    alphaBoolBlend {blendOption=Linear}
    boundBone {boneName=hair_br_03_dyn}
    componentLinearAccScale=<1 1 1>
    componentLinearVelScale=<1 1 1>
    physicsBodyDefinitions [{
      sphereCollisionRadius=0
      boundBone {boneName=hair_br_03_dyn}
      boxExtents=<10 20 20>
      constraintSetup {angularLimitsMax=<45 15 45>, angularLimitsMin=<-45 -15 -45>, angularTarget=<0 0 0>}
      localJointOffset=<1 0 -0.2>
     }]
    sphericalLimits [{limitRadius=6.5, drivingBone {boneName=head}, sphereLocalOffset=<4 0 0>}]
   }
   n15 = Anim Dynamics - Bone: hair_bl_01_dyn(ComponentPose=@n14, Alpha=1, ExternalForce="0,0,0") {
    angularDampingOverride=0.99
    angularSpringConstant=75
    bAngularSpring=true
    bGravityOverrideInSimSpace=true
    bOverrideAngularDamping=true
    bUseSphericalLimits=true
    simulationSpace=RootRelative
    alphaBoolBlend {blendOption=Linear}
    boundBone {boneName=hair_bl_01_dyn}
    componentLinearAccScale=<1 1 1>
    componentLinearVelScale=<1 1 1>
    physicsBodyDefinitions [{
      sphereCollisionRadius=0
      boundBone {boneName=hair_bl_01_dyn}
      boxExtents=<10 20 20>
      constraintSetup {angularLimitsMax=<45 15 45>, angularLimitsMin=<-45 -15 -15>, angularTarget=<0 0 0>}
      localJointOffset=<1 0 -0.2>
     }]
    sphericalLimits [{limitRadius=6.5, drivingBone {boneName=head}, sphereLocalOffset=<4 0 0>}]
   }
   n16 = Anim Dynamics - Bone: hair_bl_02_dyn(ComponentPose=@n15, Alpha=1, ExternalForce="0,0,0") {
    angularDampingOverride=0.99
    angularSpringConstant=75
    bAngularSpring=true
    bGravityOverrideInSimSpace=true
    bOverrideAngularDamping=true
    bUseSphericalLimits=true
    simulationSpace=RootRelative
    alphaBoolBlend {blendOption=Linear}
    boundBone {boneName=hair_bl_02_dyn}
    componentLinearAccScale=<1 1 1>
    componentLinearVelScale=<1 1 1>
    physicsBodyDefinitions [{
      sphereCollisionRadius=0
      boundBone {boneName=hair_bl_02_dyn}
      boxExtents=<10 20 20>
      constraintSetup {angularLimitsMax=<45 15 45>, angularLimitsMin=<-45 -15 -15>, angularTarget=<0 0 0>}
      localJointOffset=<1 0 -0.2>
     }]
    sphericalLimits [{limitRadius=6.5, drivingBone {boneName=head}, sphereLocalOffset=<4 0 0>}]
   }
   n17 = Anim Dynamics - Bone: hair_bl_03_dyn(ComponentPose=@n16, Alpha=1, ExternalForce="0,0,0") {
    angularDampingOverride=0.99
    angularSpringConstant=75
    bAngularSpring=true
    bGravityOverrideInSimSpace=true
    bOverrideAngularDamping=true
    bUseSphericalLimits=true
    simulationSpace=RootRelative
    alphaBoolBlend {blendOption=Linear}
    boundBone {boneName=hair_bl_03_dyn}
    componentLinearAccScale=<1 1 1>
    componentLinearVelScale=<1 1 1>
    physicsBodyDefinitions [{
      sphereCollisionRadius=0
      boundBone {boneName=hair_bl_03_dyn}
      boxExtents=<10 20 20>
      constraintSetup {angularLimitsMax=<45 15 45>, angularLimitsMin=<-45 -15 -15>, angularTarget=<0 0 0>}
      localJointOffset=<1 0 -0.2>
     }]
    sphericalLimits [{limitRadius=6.5, drivingBone {boneName=head}, sphereLocalOffset=<4 0 0>}]
   }
   n18 = Anim Dynamics - Bone: hair_mid_bl_01_dyn(ComponentPose=@n17, Alpha=1, ExternalForce="0,0,0") {
    angularDampingOverride=0.99
    angularSpringConstant=75
    bAngularSpring=true
    bGravityOverrideInSimSpace=true
    bOverrideAngularDamping=true
    bUseSphericalLimits=true
    simulationSpace=RootRelative
    alphaBoolBlend {blendOption=Linear}
    boundBone {boneName=hair_mid_bl_01_dyn}
    componentLinearAccScale=<1 1.5 1>
    componentLinearVelScale=<1 1.5 1>
    physicsBodyDefinitions [{
      sphereCollisionRadius=0
      boundBone {boneName=hair_mid_bl_01_dyn}
      boxExtents=<10 20 20>
      constraintSetup {angularLimitsMax=<45 45 45>, angularLimitsMin=<-45 -45 -15>, angularTarget=<0 0 0>}
      localJointOffset=<1 0 -0.2>
     }]
    sphericalLimits [{limitRadius=6.5, drivingBone {boneName=head}, sphereLocalOffset=<4 0 0>}]
   }
   n19 = Anim Dynamics - Bone: hair_mid_bl_02_dyn(ComponentPose=@n18, Alpha=1, ExternalForce="0,0,0") {
    angularDampingOverride=0.99
    angularSpringConstant=75
    bAngularSpring=true
    bGravityOverrideInSimSpace=true
    bOverrideAngularDamping=true
    bUseSphericalLimits=true
    simulationSpace=RootRelative
    alphaBoolBlend {blendOption=Linear}
    boundBone {boneName=hair_mid_bl_02_dyn}
    componentLinearAccScale=<1 1.5 1>
    componentLinearVelScale=<1 1.5 1>
    physicsBodyDefinitions [{
      sphereCollisionRadius=0
      boundBone {boneName=hair_mid_bl_02_dyn}
      boxExtents=<10 20 20>
      constraintSetup {angularLimitsMax=<45 45 45>, angularLimitsMin=<-45 -45 -15>, angularTarget=<0 0 0>}
      localJointOffset=<1 0 -0.2>
     }]
    sphericalLimits [{limitRadius=6.5, drivingBone {boneName=head}, sphereLocalOffset=<4 0 0>}]
   }
   n20 = Anim Dynamics - Bone: hair_mid_bl_03_dyn(ComponentPose=@n19, Alpha=1, ExternalForce="0,0,0") {
    angularDampingOverride=0.99
    angularSpringConstant=75
    bAngularSpring=true
    bGravityOverrideInSimSpace=true
    bOverrideAngularDamping=true
    bUseSphericalLimits=true
    simulationSpace=RootRelative
    alphaBoolBlend {blendOption=Linear}
    boundBone {boneName=hair_mid_bl_03_dyn}
    componentLinearAccScale=<1 1.5 1>
    componentLinearVelScale=<1 1.5 1>
    physicsBodyDefinitions [{
      sphereCollisionRadius=0
      boundBone {boneName=hair_mid_bl_03_dyn}
      boxExtents=<10 20 20>
      constraintSetup {angularLimitsMax=<45 45 45>, angularLimitsMin=<-45 -45 -15>, angularTarget=<0 0 0>}
      localJointOffset=<1 0 -0.2>
     }]
    sphericalLimits [{limitRadius=6.5, drivingBone {boneName=head}, sphereLocalOffset=<4 0 0>}]
   }
   n21 = Anim Dynamics - Bone: hair_mid_fl_01_dyn(ComponentPose=@n20, Alpha=1, ExternalForce="0,0,0") {
    angularDampingOverride=0.99
    angularSpringConstant=75
    bAngularSpring=true
    bGravityOverrideInSimSpace=true
    bOverrideAngularDamping=true
    bUseSphericalLimits=true
    simulationSpace=RootRelative
    alphaBoolBlend {blendOption=Linear}
    boundBone {boneName=hair_mid_fl_01_dyn}
    componentLinearAccScale=<1 2 1>
    componentLinearVelScale=<1 2 1>
    physicsBodyDefinitions [{
      sphereCollisionRadius=0
      boundBone {boneName=hair_mid_fl_01_dyn}
      boxExtents=<10 20 20>
      constraintSetup {angularLimitsMax=<45 45 45>, angularLimitsMin=<-45 -45 -15>, angularTarget=<0 0 0>}
      localJointOffset=<1 0 -0.2>
     }]
    sphericalLimits [{limitRadius=6.5, drivingBone {boneName=head}, sphereLocalOffset=<4 0 0>}]
   }
   n22 = Anim Dynamics - Bone: hair_mid_fl_02_dyn(ComponentPose=@n21, Alpha=1, ExternalForce="0,0,0") {
    angularDampingOverride=0.99
    angularSpringConstant=75
    bAngularSpring=true
    bGravityOverrideInSimSpace=true
    bOverrideAngularDamping=true
    bUseSphericalLimits=true
    simulationSpace=RootRelative
    alphaBoolBlend {blendOption=Linear}
    boundBone {boneName=hair_mid_fl_02_dyn}
    componentLinearAccScale=<1 2 1>
    componentLinearVelScale=<1 2 1>
    physicsBodyDefinitions [{
      sphereCollisionRadius=0
      boundBone {boneName=hair_mid_fl_02_dyn}
      boxExtents=<10 20 20>
      constraintSetup {angularLimitsMax=<45 45 45>, angularLimitsMin=<-45 -45 -15>, angularTarget=<0 0 0>}
      localJointOffset=<1 0 -0.2>
     }]
    sphericalLimits [{limitRadius=6.5, drivingBone {boneName=head}, sphereLocalOffset=<4 0 0>}]
   }
   n23 = Anim Dynamics - Bone: hair_mid_fl_03_dyn(ComponentPose=@n22, Alpha=1, ExternalForce="0,0,0") {
    angularDampingOverride=0.99
    angularSpringConstant=75
    bAngularSpring=true
    bGravityOverrideInSimSpace=true
    bOverrideAngularDamping=true
    bUseSphericalLimits=true
    simulationSpace=RootRelative
    alphaBoolBlend {blendOption=Linear}
    boundBone {boneName=hair_mid_fl_03_dyn}
    componentLinearAccScale=<1 2 1>
    componentLinearVelScale=<1 2 1>
    physicsBodyDefinitions [{
      sphereCollisionRadius=0
      boundBone {boneName=hair_mid_fl_03_dyn}
      boxExtents=<10 20 20>
      constraintSetup {angularLimitsMax=<45 45 45>, angularLimitsMin=<-45 -45 -15>, angularTarget=<0 0 0>}
      localJointOffset=<1 0 -0.2>
     }]
    sphericalLimits [{limitRadius=6.5, drivingBone {boneName=head}, sphereLocalOffset=<4 0 0>}]
   }
   n24 = Anim Dynamics - Bone: hair_bang_l_01_dyn(ComponentPose=@n23, Alpha=1, ExternalForce="0,0,0") {
    angularDampingOverride=0.99
    angularSpringConstant=75
    bAngularSpring=true
    bGravityOverrideInSimSpace=true
    bOverrideAngularDamping=true
    bUseSphericalLimits=true
    simulationSpace=RootRelative
    alphaBoolBlend {blendOption=Linear}
    boundBone {boneName=hair_bang_l_01_dyn}
    componentLinearAccScale=<1 2 1>
    componentLinearVelScale=<1 2 1>
    physicsBodyDefinitions [{
      sphereCollisionRadius=0
      boundBone {boneName=hair_bang_l_01_dyn}
      boxExtents=<10 20 20>
      constraintSetup {angularLimitsMax=<45 0 0>, angularLimitsMin=<-45 -45 -5>, angularTarget=<0 0 0>}
      localJointOffset=<1 0 -0.2>
     }]
    sphericalLimits [{limitRadius=6.5, drivingBone {boneName=head}, sphereLocalOffset=<4 0 0>}]
   }
   n25 = Anim Dynamics - Bone: hair_bang_l_02_dyn(ComponentPose=@n24, Alpha=1, ExternalForce="0,0,0") {
    angularDampingOverride=0.99
    angularSpringConstant=75
    bAngularSpring=true
    bGravityOverrideInSimSpace=true
    bOverrideAngularDamping=true
    bUseSphericalLimits=true
    simulationSpace=RootRelative
    alphaBoolBlend {blendOption=Linear}
    boundBone {boneName=hair_bang_l_02_dyn}
    componentLinearAccScale=<1 2 1>
    componentLinearVelScale=<1 2 1>
    physicsBodyDefinitions [{
      sphereCollisionRadius=0
      boundBone {boneName=hair_bang_l_02_dyn}
      boxExtents=<10 20 20>
      constraintSetup {angularLimitsMax=<45 0 15>, angularLimitsMin=<-45 -45 -5>, angularTarget=<0 0 0>}
      localJointOffset=<1 0 -0.2>
     }]
    sphericalLimits [{limitRadius=6.5, drivingBone {boneName=head}, sphereLocalOffset=<4 0 0>}]
   }
   n26 = Anim Dynamics - Bone: hair_bang_l_03_dyn(ComponentPose=@n25, Alpha=1, ExternalForce="0,0,0") {
    angularDampingOverride=0.99
    angularSpringConstant=75
    bAngularSpring=true
    bGravityOverrideInSimSpace=true
    bOverrideAngularDamping=true
    bUseSphericalLimits=true
    simulationSpace=RootRelative
    alphaBoolBlend {blendOption=Linear}
    boundBone {boneName=hair_bang_l_03_dyn}
    componentLinearAccScale=<1 2 1>
    componentLinearVelScale=<1 2 1>
    physicsBodyDefinitions [{
      sphereCollisionRadius=0
      boundBone {boneName=hair_bang_l_03_dyn}
      boxExtents=<10 20 20>
      constraintSetup {angularLimitsMax=<45 0 45>, angularLimitsMin=<-45 -45 -5>, angularTarget=<0 0 0>}
      localJointOffset=<1 0 -0.2>
     }]
    sphericalLimits [{limitRadius=6.5, drivingBone {boneName=head}, sphereLocalOffset=<4 0 0>}]
   }
   Output Pose(Result=Component To Local(ComponentPose=@n26)) {layerGroup=None}
  ]}
 EventGraph {nodes [Event BlueprintUpdateAnimation()]}
]
variables []
